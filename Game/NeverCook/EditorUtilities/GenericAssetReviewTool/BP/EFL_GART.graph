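# Editor utility function library for GART (Generic Asset Review Tool), supplying its report runners and asset audit helpers.
# Runners resolve report classes to CDOs, filter each asset or object, and map failures to report lists; helpers build CSV rows.
# Checks cover HLOD inclusion, FX-only mesh referencers, per-quality vertex budgets and UV-versus-material layers.
# Embedded Python reads editor properties, while NoLoad functions parse registry tags to avoid loading meshes.
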EditorUtilityBlueprint /Game/NeverCook/EditorUtilities/GenericAssetReviewTool/BP/EFL_GART
graphs [
 ActorsAreIncludedInHLOD {nodes [
   n1 = ActorsAreIncludedInHLOD!() -> (then, Actors:Array of Actor Object References) {is=title K2Node_FunctionEntry}
   > Add Items!(
    execute
    TargetSet:Set of Actor Object References=Get l_actors() -> (l_actors:Set of Actor Object References)
    NewItems:Array of Actor Object References=@n1
   ) -> (then)
   > n2 = Sequence!(execute) -> (then_0, then_1, then_2:Exec)
   > n3 = To Array!(execute, A:Set of Actor Object References=Get l_actors()) -> (then, Result:Array of Actor Object References)
   > n4 = Set l_actorArray(l_actorArray:Array of Actor Object References=@n3) -> (Output_Get:Array of Actor Object References)
   > n5 = Reverse for Each Loop!(Exec:Exec, Array:Array of Actor Object References=@n4) -> (LoopBody:Exec, ArrayIndex:int, ArrayElement:Actor Object Reference, Completed:Exec) {is=title K2Node_MacroInstance}
   > n6 = GetEditorProperty!(execute:Exec, Object:Object Reference=@n5.ArrayElement, PropertyName:Name=bEnableAutoLODGeneration) -> (then:Exec, PropertyValue:bool, ReturnValue:bool)
   > n7 = Set l_actorEnabled(l_actorEnabled:bool=@n6) -> (Output_Get:bool)
   > n8 = Branch(Condition=@n7)
   Remove Index!(
    execute=@n8.else
    TargetArray:Array of Actor Object References=Get l_actorArray() -> (l_actorArray:Array of Actor Object References)
    IndexToRemove:int=@n5
   ) -> (then)
   n9 = Execute Python Script!(
    execute=@n2.then_1
    PythonScript="import unreal\r\n\r\nfinalMap = unreal.Map(key=unreal.Name, value=int)\r\nreturn_array = []\r\nactor_list = list(in_actors)\r\ndef runCompCheck():\r\n    it = unreal.ObjectIterator(unreal.StaticMeshComponent)\r\n    \r\n    for comp in it:\r\n        actor = comp.get_outer()\r\n        if \"TRASH\" not in str(comp):\r\n            if actor in actor_list:\r\n                if isCompInHLOD(comp):\r\n                    if actor not in return_array:\r\n                        return_array.append(actor)\r\n                        actor_list.remove(actor)\r\n\r\ndef isCompInHLOD(comp):\r\n    return_value = True\r\n    if comp.get_editor_property(\"Mobility\") != unreal.ComponentMobility.STATIC:\r\n        reason = \"Mobility\"\r\n        return_value = False\r\n    \r\n    if not comp.get_editor_property(\"bEnableAutoLODGeneration\"):\r\n        reason = \"bEnableAutoLODGeneration\"\r\n        return_value = False\r\n    \r\n    if comp.is_excluded_from_hlod_level(unreal.HLODLevelExclusion.HLOD0):\r\n        reason = \"HLOD0\"\r\n        return_value = False\r\n    \r\n    if comp.is_excluded_from_hlod_level(unreal.HLODLevelExclusion.HLOD1):\r\n        reason = \"HLOD1\"\r\n        return_value = False\r\n    \r\n    if not return_value:\r\n        comp_name = unreal.SystemLibrary.get_display_name(comp)\r\n    \r\n    return return_value\r\n\r\nrunCompCheck()"
    in_actors:Array of Actor Object References=Get l_actorArray()
   ) -> (then, ReturnValue, return_array:Array of Actor Object References) {inputs [in_actors], outputs [return_array]}
   > Set l_actorArray(l_actorArray=@n9.return_array)
   n10 = Get l_actors()
   n11 = Get l_actorArray()
   Remove Items!(execute=@n2.then_2, TargetSet:Set of Actor Object References=@n10, Items:Array of Actor Object References=@n11) -> (then)
   > n12 = To Array(A=@n10)
   > Return Node!(execute, Excluded:Array of Actor Object References=@n12, Included:Array of Actor Object References=@n11)
  ]}
 CheckPathIgnoreList {nodes [
   n1 = CheckPathIgnoreList!() -> (then, AssetData:Asset Data Structure, IgnoreList:Array of Strings) {is=title K2Node_FunctionEntry}
   > n2 = Branch(Condition="integer > integer"(A=Length(TargetArray:Array of Strings=@n1.IgnoreList)))
   > n3 = For Each Loop with Break!(Exec:Exec, Array:Array of Strings=Get IgnoreList() -> (IgnoreList:Array of Strings), Break:Exec=@n5.then) -> (LoopBody:Exec, Array Element:String, Array Index:int, Completed:Exec)
   Return Node!(execute=@n3.Completed, IgnoreAsset:bool=Get l_IgnoreAssetB() -> (l_IgnoreAssetB:bool))
   n4 = "To String (Name)"(InName=Break Asset Data(AssetData=@n1).PackagePath)
   > Branch(Condition="integer > integer"(A=FindSubstring(SearchIn=@n4, Substring=@n3), B=-1))
   > n5 = Set l_IgnoreAssetB(l_IgnoreAssetB:bool=true) -> (Output_Get:bool)
   Return Node(execute=@n2.else, IgnoreAsset=false)
  ]}
 GetCSVHeaders {nodes [
   GetCSVHeaders() {is=title K2Node_FunctionEntry}
   > n1 = Set l_row(l_row:String="Type,Name,Title,Desc Short,Desc Long,Severity,Report Class") -> (Output_Get:String)
   > Return Node!(execute, RowString:String=@n1, Separator:String)
  ]}
 GetCSVRowForData {nodes [
   n1 = GetCSVRowForData!() -> (then, data:Struct GART Report Output Data Structure) {is=title K2Node_FunctionEntry}
   > Set l_separator(l_separator:String=GetCSVSeparator!(__WorldContext:Object Reference) -> (Separator:String)) -> (Output_Get:String)
   > n2 = Sequence()
   n3 = Break Struct GART Report Output Data!(Struct_GART_ReportOutputData:"Struct GART Report Output Data Structure (by ref)"=@n1) -> (
    ReportData_46_#1:Struct GART Report Data Structure
    AssetData_36_#2:Asset Data Structure
    Object_39_#3:Object Reference
    Report_44_#4:Soft Object Reference
   ) {is=title K2Node_BreakStruct}
   > n4 = Is Valid(InputObject=@n3.Object_39_#3)
   Set l_row(execute=@n4.Is Not Valid, l_row=Asset)
   n5 = Break Asset Data(AssetData=@n3.AssetData_36_#2)
   > Set l_row(l_row=Append!(A=Get l_row() -> (l_row:String), B=Get l_separator() -> (l_separator:String), C:String="To String (Name)"(InName=@n5)) -> (ReturnValue))
   Set l_row(execute=@n4.Is Valid, l_row=Object)
   > Set l_row(l_row=Append(A=Get l_row(), B=Get l_separator(), C=GetDisplayName(Object=@n3.Object_39_#3)))
   n6 = Break Struct GART Report Data!(Struct_GART_ReportData:"Struct GART Report Data Structure (by ref)"=@n3) -> (
    DisplayTitle_20_#5:String
    DisplayDescShort_29_#6:String
    DisplayDescLong_30_#7:String
    DisplayColorOverride_40_#8:Linear Color Structure
    Tags_25_#9:Array of Names
    Severity_46_#1:Enum_GART_ReportSeverity Enum=~
   ) {is=title K2Node_BreakStruct}
   Set l_row(execute=@n2.then_1, l_row=Append(A=Get l_row(), B=Get l_separator(), C=@n6))
   > Set l_row(l_row=Append(A=Get l_row(), B=Get l_separator(), C=@n6.DisplayDescShort_29_#6))
   > Set l_row(l_row=Append(A=Get l_row(), B=Get l_separator(), C=@n6.DisplayDescLong_30_#7))
   > Set l_row(l_row=Append(
     A=Get l_row()
     B=Get l_separator()
     C=Enum to String!(Enumerator:Enum_GART_ReportSeverity Enum=@n6.Severity_46_#1) -> (ReturnValue:String)
    ))
   > Set l_row(l_row=Append(
     A=Get l_row()
     B=Get l_separator()
     C="To String (SoftObjectReference)"!(SoftObjectReference:"Soft Object Reference (by ref)"=@n3.Report_44_#4) -> (ReturnValue:String)
    ))
   Return Node!(execute=@n2.then_2, RowString=Get l_row())
  ]}
 GetCSVSeparator {nodes [
   GetCSVSeparator() {is=title K2Node_FunctionEntry}
   > Return Node!(execute, Separator:String=|)
  ]}
 GetEditorPropertyMap {nodes [
   n1 = GetEditorPropertyMap!() -> (then, AssetData:Asset Data Structure) {is=title K2Node_FunctionEntry}
   > n2 = Execute Python Script!(
    execute
    PythonScript="import unreal\r\n\r\nassetObject = unreal.load_object(None, asset_object_path)\r\n\r\ncdo = unreal.get_default_object(assetObject)\r\nproperty_dict = {}\r\nattInitKeys = cdo.get_editor_property(\"AttributeInitKeys\")\r\n\r\nproperty_dict['ResourceType'] = str(cdo.get_editor_property(\"ResourceType\").name)\r\nproperty_dict['Attributes'] = str(attInitKeys[0].get_editor_property(\"attributeInitCategory\")) + str(attInitKeys[0].get_editor_property(\"attributeInitSubCategory\")) + \", \" + str(attInitKeys[1].get_editor_property(\"attributeInitCategory\")) + str(attInitKeys[1].get_editor_property(\"attributeInitSubCategory\"))\r\nproperty_dict['ResourceAmount'] = str(cdo.get_editor_property(\"BuildingResourceAmountOverride\").get_editor_property(\r\n    \"RowName\"))\r\nproperty_dict['BuildingTypeEnum'] = str(cdo.get_editor_property(\"BuildingType\").name)\r\nproperty_dict['AttachmentSlotEnum'] = str(cdo.get_editor_property(\"building_attachment_slot\").name)\r\nproperty_dict['AttachmentTypeEnum'] = str(cdo.get_editor_property(\"building_attachment_type\").name)\r\n\r\nif cdo.get_editor_property(\"HotSpotConfig\") is not None:\r\n    property_dict['HotSpotConfig'] = str(cdo.get_editor_property(\"HotSpotConfig\").get_name())\r\nelse:\r\n    property_dict['HotSpotConfig'] = \"None\"\r\n\r\nif cdo.get_editor_property(\"AreaClass\") is not None:\r\n    property_dict['AreaClass'] = str(cdo.get_editor_property(\"AreaClass\").get_name())\r\nelse:\r\n    property_dict['AreaClass'] = \"None\"\r\n\r\nif cdo.get_editor_property(\"NavigationLinkSoftClass\") is not None:\r\n    property_dict['NavLinkClass'] = str(cdo.get_editor_property(\"NavigationLinkSoftClass\").get_name())\r\nelse:\r\n    property_dict['NavLinkClass'] = \"None\"\r\n\r\nproperty_dict['BrMaxDrawDistance'] = str(cdo.get_editor_property(\"BRMaxDrawDistance\"))\r\nproperty_dict['CreatePhysicsObjectComponent'] = str(cdo.get_editor_property(\"bCreatePhysicsObjectComponent\"))\r\nproperty_dict['DestroyOnPlayerBuildingPlacement'] = str(cdo.get_editor_property(\"bDestroyOnPlayerBuildingPlacement\"))\r\nproperty_dict['CanBeDamaged'] = str(cdo.get_editor_property(\"bCanBeDamaged\"))\r\nproperty_dict['EnableAutoLODGeneration'] = str(cdo.get_editor_property(\"bEnableAutoLODGeneration\"))\r\n\r\nif cdo.get_editor_property(\"EditModePatternData\") is not None:\r\n    property_dict['EditModePatternData'] = str(cdo.get_editor_property(\"EditModePatternData\").get_name())\r\nelse:\r\n    property_dict['EditModePatternData'] = \"None\"\r\n\r\nproperty_dict['IncludeActorInHLOD'] = str(cdo.get_editor_property(\"bEnableAutoLODGeneration\"))\r\nproperty_dict['ExcludeFromHLODLevels'] = str(cdo.get_editor_property('StaticMeshComponent').get_editor_property(\r\n    \"ExcludeFromHLODLevels\"))\r\nproperty_dict['IncludeComponentInHLOD'] = str(cdo.get_editor_property('StaticMeshComponent').get_editor_property(\r\n    \"bEnableAutoLODGeneration\"))\r\n\r\n##print(property_dict)"
    asset_object_path:String=Append!(A=ToSoftObjectPath!(InAssetData:"Asset Data Structure (by ref)"=@n1) -> (ReturnValue_PathString:String), B=_C) -> (ReturnValue)
   ) -> (then, ReturnValue, property_dict:Map of Strings to Strings) {inputs [asset_object_path], outputs [property_dict]}
   > Return Node!(execute, Out Editor Properties:Map of Strings to Strings=@n2.property_dict)
  ]}
 GetHardRefsToReportClasses {nodes [
   n1 = GetHardRefsToReportClasses!() -> (then, Array:Array of GART Report Base Soft Class References) {is=title K2Node_FunctionEntry}
   > n2 = For Each Loop(Array:Array of GART Report Base Soft Class References=@n1) -> (Array Element:GART Report Base Soft Class Reference)
   Return Node!(
    execute=@n2.Completed
    ReportObjects:Array of GART Report Base Object References=Get l_allCdos() -> (l_allCdos:Array of GART Report Base Object References)
   )
   > n3 = LoadClassAsset_Blocking(AssetClass=@n2)
   > n4 = Execute Python Script!(execute, PythonScript="import unreal\r\nreport_cdo = unreal.get_default_object(report_class)", report_class:Object Class Reference=@n3) -> (then, ReturnValue, report_cdo:Object Reference) {inputs [report_class], outputs [report_cdo]}
   > n5 = Cast To GART_Report_Base!(execute, Object=@n4.report_cdo) -> (then, CastFailed, AsGART Report Base:GART Report Base Object Reference) {is=title K2Node_DynamicCast}
   > Add Unique(TargetArray:Array of GART Report Base Object References=Get l_allCdos(), NewItem:"GART Report Base Object Reference (by ref)"=@n5)
  ]}
 IsMeshOnlyUsedInFX {nodes [
   n1 = MakeTopLevelAssetPath!(PackageName:String=/Script/Engine, AssetName:String=Blueprint) -> (ReturnValue:Top Level Asset Path Structure)
   n2 = RunAssetsThroughFilter!(
    execute
    self:Asset Registry Interface=GetAssetRegistry()
    AssetDataList:Array of Asset Data Structures=Get l_bps() -> (l_bps:Array of Asset Data Structures)
    Filter:"ARFilter Structure (by ref)"=MakeARFilter!(
     PackageNames:Array of Names
     PackagePaths:Array of Names
     SoftObjectPaths:Array of Soft Object Path Structures
     ClassPaths:Array of Top Level Asset Path Structures=Make Array("[0]":Top Level Asset Path Structure=@n1) -> (Array:Array of Top Level Asset Path Structures)
     RecursiveClassPathsExclusionSet:Set of Top Level Asset Path Structures
     ClassNames:Array of Names
     RecursiveClassesExclusionSet:Set of Names
     bRecursivePaths:bool=false
     bRecursiveClasses:bool=false
     bIncludeOnlyOnDiskAssets:bool=false
    ) -> (ReturnValue:ARFilter Structure)
   ) -> (then)
   n3 = Sequence!(execute=[@n2.then, @n13.then]) -> (then_0, then_1)
   n4 = Get l_fx() -> (l_fx:Array of Asset Data Structures)
   > Branch(Condition="integer > integer"(A=Length(TargetArray:Array of Asset Data Structures=@n4)))
   > Set l_hasFxRef(l_hasFxRef:bool=true) -> (Output_Get:bool)
   n5 = Length(TargetArray=Get l_allRefs() -> (l_allRefs:Array of Asset Data Structures))
   n6 = Length(TargetArray=Get l_fx())
   Return Node!(execute=@n3.then_1, OnlyFX:bool=AND Boolean(A="Equal (Integer)"(A=@n6, B=@n5), B=Get l_hasFxRef() -> (l_hasFxRef:bool)))
   n7 = IsMeshOnlyUsedInFX!() -> (then, assetData:Asset Data Structure) {is=title K2Node_FunctionEntry}
   > Set l_data(l_data:Asset Data Structure=@n7) -> (Output_Get:Asset Data Structure)
   > n8 = Get Referencers!(
    execute
    self:Asset Registry Interface=GetAssetRegistry()
    PackageName:Name=Break Asset Data!(AssetData=Get l_data() -> (l_data:Asset Data Structure)) -> (PackageName)
    ReferenceOptions:"Asset Registry Dependency Options Structure (by ref)"=Make Asset Registry Dependency Options(bIncludeSoftPackageReferences=True)
   ) -> (then, OutReferencers:Array of Names, ReturnValue:bool) {is=title K2Node_CallFunction}
   > Branch(Condition=Is Valid Index!(TargetArray:Array of Names=@n8, IndexToTest:int) -> (ReturnValue:bool))
   > n9 = GetAssets!(
    execute
    self:Asset Registry Interface=GetAssetRegistry()
    Filter:"ARFilter Structure (by ref)"=MakeARFilter(PackageNames=@n8, bRecursivePaths=false, bRecursiveClasses=false, bIncludeOnlyOnDiskAssets=false)
    bSkipARFilteredAssets:bool=true
   ) -> (then, OutAssetData:Array of Asset Data Structures, ReturnValue:bool)
   > n10 = Set l_allRefs(l_allRefs:Array of Asset Data Structures=@n9) -> (Output_Get:Array of Asset Data Structures)
   > n11 = Set l_fx(l_fx:Array of Asset Data Structures=@n10) -> (Output_Get:Array of Asset Data Structures)
   > Set l_bps(l_bps:Array of Asset Data Structures=@n11) -> (Output_Get:Array of Asset Data Structures)
   n12 = MakeTopLevelAssetPath(PackageName=&Engine, AssetName=Niagara)
   > n13 = RunAssetsThroughFilter(
    self=GetAssetRegistry()
    AssetDataList=Get l_fx()
    Filter=MakeARFilter(ClassPaths=Make Array("[0]"=@n12), bRecursivePaths=false, bRecursiveClasses=false, bIncludeOnlyOnDiskAssets=false)
   )
   Break Asset Data!(AssetData=For Each Loop(Array:Array of Asset Data Structures=Get l_fx()) -> (Array Element:Asset Data Structure)) -> (PackageName, PackagePath:Name, AssetName:Name, AssetClassPath:Top Level Asset Path Structure)
  ]}
 IsMeshWithinVertBudget {nodes [
   n1 = GetLodCount!(execute, self:Static Mesh Editor Subsystem Object Reference=Get StaticMeshEditorSubsystem(), StaticMesh:Static Mesh Object Reference) -> (then, ReturnValue:int=0)
   > For Loop(LastIndex=int - int(A=@n1))
   n2 = IsMeshWithinVertBudget!() -> (then, mesh:Static Mesh Object Reference, QualityBudgets:Map of EPerQualityLevels Enums to Integers=Low) {is=title K2Node_FunctionEntry}
   > n3 = LoadMeshAndGetMinLODVertexCounts!(execute, mesh:Static Mesh Object Reference=@n2, __WorldContext:Object Reference) -> (then, QualityPerVertCountMap:Map of EPerQualityLevels Enums to Integers=Low)
   > Set l_foundMinLODQuality(l_foundMinLODQuality:Map of EPerQualityLevels Enums to Integers=@n3) -> (Output_Get:Map of EPerQualityLevels Enums to Integers=Low)
   n4 = Get QualityBudgets() -> (QualityBudgets:Map of EPerQualityLevels Enums to Integers=Low)
   > n5 = Keys!(execute, TargetMap:Map of EPerQualityLevels Enums to Integers=@n4) -> (then, Keys:Array of EPerQualityLevels Enums)
   > n6 = Values!(execute, TargetMap:Map of EPerQualityLevels Enums to Integers=@n4) -> (then, Values:Array of Integers)
   > n7 = For Each Loop(Array:Array of EPerQualityLevels Enums=@n5) -> (Array Element:EPerQualityLevels Enum)
   Return Node!(
    execute=@n7.Completed
    BadQualityMap:Map of EPerQualityLevels Enums to Vector 2D Structures=Get l_return() -> (l_return:Map of EPerQualityLevels Enums to Vector 2D Structures=Low)
   )
   n8 = Find(
    TargetMap:Map of EPerQualityLevels Enums to Integers=Get l_foundMinLODQuality() -> (l_foundMinLODQuality:Map of EPerQualityLevels Enums to Integers=Low)
    Key:"EPerQualityLevels Enum (by ref)"=@n7
   ) -> (Value:int)
   n9 = "Get (a ref)"(Array=@n6, Dimension 1=@n7.Array Index)
   > Branch(Condition="integer < integer"(A=@n9, B=@n8))
   > Add!(
    execute
    TargetMap:Map of EPerQualityLevels Enums to Vector 2D Structures=Get l_return() -> (l_return=Low)
    Key:"EPerQualityLevels Enum (by ref)"=@n7
    Value:"Vector 2D Structure (by ref)"=MakeVector2D(X="To Float (Integer)"(InInt=@n8), Y="To Float (Integer)"(InInt=@n9))
   ) -> (then) {is=title K2Node_CallFunction}
  ]}
 LoadAssetAndGetVertexCounts {nodes [
   n1 = LoadAssetAndGetVertexCounts!() -> (then, AssetData:"Asset Data Structure (by ref)") {is=title K2Node_FunctionEntry}
   > n2 = Cast To StaticMesh(Object=GetAsset(InAssetData=@n1))
   > Set l_mesh(l_mesh:Static Mesh Object Reference=@n2) -> (Output_Get:Static Mesh Object Reference)
   > n3 = GetLodCount(self=Get StaticMeshEditorSubsystem(), StaticMesh=@n2) -> (ReturnValue=0)
   > n4 = For Loop(LastIndex=int - int(A=@n3))
   Return Node!(execute=@n4.Completed, VertsPerLOD:Array of Integers=Get l_returns() -> (l_returns:Array of Integers))
   > Add(
    TargetArray:Array of Integers=Get l_returns()
    NewItem:"Integer (by ref)"=GetNumberVerts!(
     self:Static Mesh Editor Subsystem Object Reference=Get StaticMeshEditorSubsystem()
     StaticMesh:Static Mesh Object Reference=Get l_mesh() -> (l_mesh:Static Mesh Object Reference)
     LODIndex:int=@n4
    ) -> (ReturnValue:int=0)
   ) {is=title K2Node_CallArrayFunction}
  ]}
 LoadMeshAndGetMinLODVertexCounts {nodes [
   n1 = GetLodCount(self=Get StaticMeshEditorSubsystem(), StaticMesh=Get l_mesh()) -> (ReturnValue=0)
   > For Loop(LastIndex=int - int(A=@n1))
   n2 = LoadMeshAndGetMinLODVertexCounts!() -> (then, mesh:Static Mesh Object Reference) {is=title K2Node_FunctionEntry}
   > Set l_mesh(l_mesh=@n2)
   n3 = GetMinLODForQualityLevels!(self:Static Mesh Object Reference=Get l_mesh()) -> (QualityLevelMinimumLODs:Map of EPerQualityLevels Enums to Integers=Low, Default:int=0)
   > n4 = Set l_foundMinLODQuality(l_foundMinLODQuality=@n3) -> (Output_Get=Low)
   > Add(TargetMap:Map of EPerQualityLevels Enums to Integers=@n4, Key=Num, Value:"Integer (by ref)"=@n3.Default) {is=title K2Node_CallFunction}
   > n5 = ForEach EPerQualityLevels!(execute:Exec, SkipHidden:bool) -> (LoopBody:Exec, EnumValue:EPerQualityLevels Enum, then:Exec) {is=title K2Node_ForEachElementInEnum}
   Return Node!(
    execute=@n5.then
    QualityPerVertCountMap:Map of EPerQualityLevels Enums to Integers=Get l_outQualityPerVertCountMap() -> (l_outQualityPerVertCountMap:Map of EPerQualityLevels Enums to Integers=Low)
   )
   n6 = Find(TargetMap=Get l_foundMinLODQuality() -> (l_foundMinLODQuality=Low), Key=@n5)
   > n7 = Branch(Condition=@n6.ReturnValue)
   > Add(
    TargetMap=Get l_outQualityPerVertCountMap() -> (l_outQualityPerVertCountMap=Low)
    Key=@n5
    Value=GetNumberVerts(self=Get StaticMeshEditorSubsystem(), StaticMesh=Get l_mesh(), LODIndex=@n6) -> (ReturnValue=0)
   )
   Add(
    execute=@n7.else
    TargetMap=Get l_outQualityPerVertCountMap() -> (l_outQualityPerVertCountMap=Low)
    Key=@n5
    Value=GetNumberVerts(self=Get StaticMeshEditorSubsystem(), StaticMesh=Get l_mesh(), LODIndex=0) -> (ReturnValue=0)
   )
  ]}
 Load_StaticMesh4LayerCheck {nodes [
   n1 = Load_StaticMesh4LayerCheck!() -> (then, mesh:Static Mesh Object Reference) {is=title K2Node_FunctionEntry}
   > n2 = Set l_mesh(l_mesh=@n1)
   > Is Valid(InputObject=@n2)
   > n3 = Sequence!(execute) -> (then_0, then_1, then_2:Exec)
   > n4 = GetClassDefaultObject!(execute, Class:Object Class Reference=&/GenMesh_GART.GenMesh_GART_C) -> (then, ReturnValue:Object Reference)
   > n5 = Cast To GeneratedDynamicMeshActor!(execute, Object=@n4) -> (then, CastFailed, AsGenerated Dynamic Mesh Actor:Generated Dynamic Mesh Actor Object Reference) {is=title K2Node_DynamicCast}
   > n6 = Set l_genMeshActor(l_genMeshActor:Generated Dynamic Mesh Actor Object Reference=@n5) -> (Output_Get:Generated Dynamic Mesh Actor Object Reference)
   > n7 = AllocateComputeMesh!(execute, self:Dynamic Mesh Actor Object Reference=@n6) -> (then, ReturnValue:Dynamic Mesh Object Reference)
   > n8 = Set l_computeMesh(l_computeMesh:Dynamic Mesh Object Reference=@n7) -> (Output_Get:Dynamic Mesh Object Reference)
   > n9 = Copy Mesh From Static Mesh with Section Materials!(
    execute
    FromStaticMeshAsset:Static Mesh Object Reference=Get l_mesh()
    ToDynamicMesh:Dynamic Mesh Object Reference=@n8
    AssetOptions:Geometry Script Copy Mesh From Asset Options Structure
    RequestedLOD:Geometry Script Mesh Read LOD Structure
   ) -> (Failure:Exec, Success:Exec, ReturnValue:Dynamic Mesh Object Reference)
   n10 = GetUVSetBoundingBox!(TargetMesh:Dynamic Mesh Object Reference=@n7, UVSetIndex:int=1) -> (bIsValidUVSet:bool, bUVSetIsEmpty:bool, ReturnValue:Box 2D Structure)
   n11 = Branch(execute=@n9.Success, Condition=@n10.bUVSetIsEmpty)
   n12 = Break Box 2D!(Box2D:"Box 2D Structure (by ref)"=@n10.ReturnValue) -> (Min:Vector 2D Structure, Max:Vector 2D Structure, bIsValid:bool=~) {is=title K2Node_BreakStruct}
   Set l_UVMax(execute=@n11.else, l_UVMax:int=Ceil(A=BreakVector2D(InVec=@n12.Max)) {is=title K2Node_CallFunction}) -> (Output_Get:int=0)
   ReleaseComputeMesh!(
    execute=@n3.then_1
    self:Dynamic Mesh Actor Object Reference=Get l_genMeshActor() -> (l_genMeshActor:Generated Dynamic Mesh Actor Object Reference)
    Mesh:Dynamic Mesh Object Reference=Get l_computeMesh() -> (l_computeMesh:Dynamic Mesh Object Reference)
   ) -> (then, ReturnValue:bool)
   Branch(execute=@n3.then_2, Condition="integer > integer"(A=Get l_UVMax() -> (l_UVMax:int=0), B=-1))
   n13 = Make Map!(
    Key 0:Name=MakeLiteralName(Value=Use 2 Layers)
    Value 0:int=MakeLiteralInt!(Value:int=2) -> (ReturnValue:int=0)
    Key 1:Name=MakeLiteralName(Value=Use 3 Layers)
    Value 1:int=MakeLiteralInt(Value=3) -> (ReturnValue=0)
    Key 2:Name=MakeLiteralName(Value=Use 4 Layers)
    Value 2:int=MakeLiteralInt(Value=4) -> (ReturnValue=0)
   ) -> (Map:Map of Names to Integers) {is=title K2Node_MakeMap}
   > n14 = Values(TargetMap:Map of Names to Integers=@n13)
   > n15 = Keys(TargetMap:Map of Names to Integers=@n13) -> (Keys:Array of Names)
   > n16 = For Each Loop with Break(Array:Array of Names=@n15) -> (Array Element:Name)
   > Set l_param(l_param:Name=@n16) -> (Output_Get:Name=None)
   > Set l_maxLayer(l_maxLayer:int="Get (a copy)"(Array:Array of Integers=@n14, Dimension 1=@n16.Array Index) -> (Output:int)) -> (Output_Get:int=0)
   > n17 = GetSectionMaterialListFromStaticMesh!(execute, FromStaticMeshAsset:Static Mesh Object Reference=Get l_mesh(), RequestedLOD:Geometry Script Mesh Read LOD Structure) -> (
    Failure:Exec
    Success:Exec
    MaterialList:Array of Material Interface Object References
    MaterialIndex:Array of Integers
    MaterialSlotNames:Array of Names
   )
   n18 = For Each Loop(Exec=@n17.Success, Array:Array of Material Interface Object References=@n17) -> (Array Element:Material Interface Object Reference)
   > n19 = Cast To MaterialInstanceConstant(Object=@n18)
   n20 = GetMaterialInstanceStaticSwitchParameterValue!(
    Instance:Material Instance Constant Object Reference=@n19
    ParameterName:Name=Get l_param() -> (l_param:Name=None)
    Association:EMaterialParameterAssociation Enum=GlobalParameter
   ) -> (ReturnValue:bool)
   > Branch(Condition=@n20)
   n21 = Get l_param() -> (l_param=None)
   > PrintString(InString=Append!(
     A="Checking "
     B=GetDisplayName(Object=@n19)
     C:String=" switch "
     D:String="To String (Name)"(InName=@n21)
     E:String=" == "
     F:String="To String (Boolean)"(InBool=@n20)
    ) -> (ReturnValue))
   n22 = Get l_UVMax() -> (l_UVMax=0)
   n23 = Get l_maxLayer() -> (l_maxLayer:int=0)
   > PrintString(InString=Append!(A="UV MAx = ", B="To String (Integer)"(InInt=@n22), C=", material == ", D="To String (Integer)"(InInt=@n23)) -> (ReturnValue))
   > Branch(Condition="integer < integer"(A=@n22, B=@n23))
   n24 = Get l_param() -> (l_param=None)
   n25 = Get l_UVMax() -> (l_UVMax=0)
   n26 = Get l_mesh()
   > PrintString(InString=Append!(
     A="To String (Name)"(InName=@n24)
     B=" is True but the mesh "
     C=GetDisplayName(Object=@n26)
     D=" only uses "
     E:String="To String (Integer)"(InInt=@n25)
     F:String=" Layers."
    ) -> (ReturnValue))
   Get StaticMeshEditorSubsystem()
  ]}
 NoLoad_GetMinLODLevelFromAssetData {nodes [
   n1 = NoLoad_GetMinLODLevelFromAssetData!() -> (then, AssetData:Asset Data Structure) {is=title K2Node_FunctionEntry}
   n2 = GetTagValue(InAssetData=@n1, InTagName=MakeLiteralName(Value=QualityLevelMinLOD))
   > Branch(Condition=@n2.ReturnValue)
   > n3 = For Each Loop(Array:Array of Strings=ParseIntoArray(SourceString=@n2, Delimiter=", ")) -> (Array Element:String)
   n4 = ForEach EPerQualityLevels(execute=@n3.Completed)
   Return Node!(
    execute=@n4.then
    QualitiesInUse:Map of EPerQualityLevels Enums to Integers=Get l_returnSorted() -> (l_returnSorted:Map of EPerQualityLevels Enums to Integers=Low)
   )
   n5 = Find(TargetMap=Get l_foundQualities() -> (l_foundQualities:Map of EPerQualityLevels Enums to Integers=Low), Key=@n4)
   > Branch(Condition=@n5.ReturnValue)
   > Add(TargetMap=Get l_returnSorted() -> (l_returnSorted=Low), Key=@n4, Value=@n5)
   n6 = Set l_string(execute=@n3.LoopBody, l_string:String=@n3) -> (Output_Get:String)
   n7 = Split(SourceString=@n6, InStr="=", SearchDir=FromStart)
   > n8 = Branch(Condition=@n7.ReturnValue)
   > n9 = Switch on String!(execute:Exec, Selection:String=@n7) -> (Default:Exec, Low:Exec, Medium:Exec, High:Exec, Epic:Exec) {pinNames [Low,Medium,High,Epic]}
   n10 = Set l_quality(execute=@n9.Epic, l_quality:EPerQualityLevels Enum=Epic) -> (Output_Get:EPerQualityLevels Enum=Low)
   Add(
    execute=[@n10.then, @n11.then, @n12.then, @n13.then]
    TargetMap=Get l_foundQualities() -> (l_foundQualities=Low)
    Key=Get l_quality() -> (l_quality:EPerQualityLevels Enum=Low)
    Value=String To Integer(InString=@n7.RightS)
   )
   n11 = Set l_quality(execute=@n9.High, l_quality=High) -> (Output_Get=Low)
   n12 = Set l_quality(execute=@n9.Low, l_quality=Low) -> (Output_Get=Low)
   n13 = Set l_quality(execute=@n9.Medium, l_quality=Medium) -> (Output_Get=Low)
   Add(
    execute=@n8.else
    TargetMap=Get l_foundQualities() -> (l_foundQualities=Low)
    Key=Num
    Value=String To Integer(InString=Get l_string() -> (l_string:String))
   )
  ]}
 NoLoad_GetStaticMeshCollisionData {nodes [
   n1 = NoLoad_GetStaticMeshCollisionData!() -> (then, AssetData:Asset Data Structure) {is=title K2Node_FunctionEntry}
   > Set l_assetData(l_assetData=@n1)
   > n2 = Sequence!(execute) -> (then_0, then_1, then_2, then_3:Exec)
   n3 = GetTagValue(InAssetData=Get l_assetData(), InTagName=MakeLiteralName(Value=SectionsWithCollision))
   > Branch(Condition=@n3.ReturnValue)
   > Set l_sectionsWithCollision(l_sectionsWithCollision:int=String To Integer(InString=@n3)) -> (Output_Get:int=0)
   n4 = GetTagValue(InAssetData=Get l_assetData(), InTagName=MakeLiteralName(Value=CollisionComplexity))
   Branch(execute=@n2.then_1, Condition=@n4.ReturnValue)
   > Set l_complexity(l_complexity:String=@n4) -> (Output_Get:String)
   n5 = GetTagValue(InAssetData=Get l_assetData(), InTagName=MakeLiteralName(Value=CollisionPrims))
   Branch(execute=@n2.then_2, Condition=@n5.ReturnValue)
   > Set l_collisionPrimCount(l_collisionPrimCount:int=String To Integer(InString=@n5)) -> (Output_Get:int=0)
   Return Node!(
    execute=@n2.then_3
    SectionsWithCollision:int=Get l_sectionsWithCollision() -> (l_sectionsWithCollision:int=0)
    CollisionPrimCount:int=Get l_collisionPrimCount() -> (l_collisionPrimCount:int=0)
    CollisionComplexity:String=Get l_complexity() -> (l_complexity:String)
   )
  ]}
 NoLoad_GetStaticMeshVertCountFromAssetData {nodes [
   n1 = NoLoad_GetStaticMeshVertCountFromAssetData!() -> (then, AssetData:Asset Data Structure) {is=title K2Node_FunctionEntry}
   n2 = GetTagValue(InAssetData=@n1, InTagName=MakeLiteralName(Value=Vertices))
   > n3 = Branch(Condition=@n2.ReturnValue)
   > Return Node!(execute, VertCount:int=String To Integer(InString=Replace(SourceString=@n2, From=",")))
   Return Node(execute=@n3.else, VertCount=-1)
  ]}
 RunAssetsThroughFilters {nodes [
   n1 = RunAssetsThroughFilters!() -> (then, AssetDatas:Array of Asset Data Structures, ReportClasses:Array of GART Report Base Soft Class References, AllowLoading:bool=~) {is=title K2Node_FunctionEntry}
   > Set l_assetDatas(l_assetDatas:Array of Asset Data Structures=@n1) -> (Output_Get:Array of Asset Data Structures)
   > n2 = GetHardRefsToReportClasses!(execute, Array:Array of GART Report Base Soft Class References=@n1.ReportClasses, __WorldContext:Object Reference) -> (then, ReportObjects:Array of GART Report Base Object References) {is=title K2Node_CallFunction}
   > Set l_reports(l_reports:Array of GART Report Base Object References=@n2) -> (Output_Get:Array of GART Report Base Object References)
   > n3 = For Each Loop(Array:Array of Asset Data Structures=Get l_assetDatas() -> (l_assetDatas:Array of Asset Data Structures)) -> (Array Element:Asset Data Structure)
   Return Node!(
    execute=@n3.Completed
    ReportDataMap:Map of Names to Struct GART Asset Report List Structures=Get l_ReportDataMap() -> (l_ReportDataMap:Map of Names to Struct GART Asset Report List Structures)
   )
   > Set l_assetData(l_assetData=@n3)
   > Clear(TargetArray:Array of Struct GART Report Output Data Structures=Get l_reportList() -> (l_reportList:Array of Struct GART Report Output Data Structures))
   > n4 = For Each Loop(Array:Array of GART Report Base Object References=Get l_reports() -> (l_reports:Array of GART Report Base Object References)) -> (Array Element:GART Report Base Object Reference)
   Branch(execute=@n4.Completed, Condition=Is Valid Index(TargetArray:Array of Struct GART Report Output Data Structures=Get l_reportList()))
   > Add(
    TargetMap:Map of Names to Struct GART Asset Report List Structures=Get l_ReportDataMap()
    Key:"Name (by ref)"=Break Asset Data!(AssetData=Get l_assetData()) -> (PackageName)
    Value:"Struct GART Asset Report List Structure (by ref)"=Make Struct GART Asset Report List!(
     ReportDatas_11_#10:Array of Struct GART Report Output Data Structures=Get l_reportList()
     AssetData_7_#11:Asset Data Structure=Get l_assetData()
    ) -> (Struct_GART_AssetReportList:Struct GART Asset Report List Structure) {is=title K2Node_MakeStruct}
   )
   n5 = InternalFilterAssetData!(
    execute=@n4.LoopBody
    self:GART Report Base Object Reference=@n4
    AssetData:Asset Data Structure=Get l_assetData()
    AllowLoading:bool=Get AllowLoading() -> (AllowLoading:bool)
   ) -> (then, Valid:bool, ReportData:Struct GART Report Output Data Structure)
   > n6 = Branch(Condition=@n5)
   Add(
    execute=@n6.else
    TargetArray:Array of Struct GART Report Output Data Structures=Get l_reportList()
    NewItem:"Struct GART Report Output Data Structure (by ref)"=@n5.ReportData
   ) {is=title K2Node_CallArrayFunction}
  ]}
 RunObjectsThroughFilters {nodes [
   n1 = RunObjectsThroughFilters!() -> (then, Objects:Array of Object References, ReportClasses:Array of GART Report Base Soft Class References) {is=title K2Node_FunctionEntry}
   > Set l_objects(l_objects:Array of Object References=@n1) -> (Output_Get:Array of Object References)
   > n2 = GetHardRefsToReportClasses(Array=@n1.ReportClasses)
   > Set l_reports(l_reports=@n2)
   > n3 = For Each Loop(Array:Array of Object References=Get l_objects() -> (l_objects:Array of Object References)) -> (Array Element:Object Reference)
   Return Node(
    execute=@n3.Completed
    ReportDataMap:Map of Object References to Struct GART Asset Report List Structures=Get l_ReportDataMap() -> (l_ReportDataMap:Map of Object References to Struct GART Asset Report List Structures)
   )
   > Set l_object(l_object:Object Reference=@n3) -> (Output_Get:Object Reference)
   > Clear(TargetArray=Get l_reportList())
   > n4 = For Each Loop(Array:Array of GART Report Base Object References=Get l_reports()) -> (Array Element:GART Report Base Object Reference)
   Branch(execute=@n4.Completed, Condition=Is Valid Index(TargetArray=Get l_reportList()))
   > Add(
    TargetMap:Map of Object References to Struct GART Asset Report List Structures=Get l_ReportDataMap()
    Key:"Object Reference (by ref)"=Get l_object() -> (l_object:Object Reference)
    Value=Make Struct GART Asset Report List!(
     ReportDatas_11_#10=Get l_reportList()
     Object_10_#12:Object Reference=Get l_object()
    ) -> (Struct_GART_AssetReportList)
   ) {is=title K2Node_CallFunction}
   n5 = InternalFilterObject!(execute=@n4.LoopBody, self:GART Report Base Object Reference=@n4, Object:Object Reference=Get l_object()) -> (then, Valid:bool, ReportData:Struct GART Report Output Data Structure)
   > n6 = Branch(Condition=@n5)
   Add(execute=@n6.else, TargetArray=Get l_reportList(), NewItem=@n5.ReportData) {is=title K2Node_CallArrayFunction}
  ]}
]
variables []
