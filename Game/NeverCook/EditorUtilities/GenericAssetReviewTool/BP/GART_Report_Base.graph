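# Base editor-utility report for the Generic Asset Review Tool, providing stub filter and click functions for child reports to override.
# Internal filters gate assets by exact class path and objects by inheritance, clear the error text, then run the filter logic.
# Post-filtering stamps the report reference, fills default colour from severity and copies accumulated, indented errors into the long description.
# InternalValidateSeverity exists, but post-filtering never calls it.
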EditorUtilityBlueprint /Game/NeverCook/EditorUtilities/GenericAssetReviewTool/BP/GART_Report_Base
graphs [
 AddError {nodes [
   n1 = AddError!() -> (then, NewError:String, Depth:int) {is=title K2Node_FunctionEntry}
   > n2 = Branch(Condition=IsEmpty(InString=@n1))
   n3 = Sequence!(execute=@n2.else) -> (then_0, then_1, then_2:Exec)
   > Branch(Condition="integer > integer"(A=Get Depth() -> (Depth:int=0)))
   > For Loop(LastIndex=Get Depth() -> (Depth=0))
   > Set l_indent(l_indent:String=Append(A=Get l_indent() -> (l_indent:String), B="  ")) -> (Output_Get:String)
   n4 = Len(S=Get CurrentErrorString() -> (CurrentErrorString:String))
   Set l_newline(execute=@n3.then_1, l_newline:String=SelectString(A="\r\n", bPickA="integer > integer"(A=@n4))) -> (Output_Get:String)
   n5 = "To Integer (Boolean)"!(InBool:bool="integer > integer"(A=Get Depth() -> (Depth=0))) -> (ReturnValue:int=0)
   > Set l_newError(l_newError:String=Append!(
     A=Get l_indent()
     B=Select(Option 0:String="🔶", Option 1:String="🔸", Index:int=@n5) -> (ReturnValue:String)
     C:String=Get NewError() -> (NewError:String)
    ) -> (ReturnValue)) -> (Output_Get:String)
   Set CurrentErrorString(
    execute=@n3.then_2
    CurrentErrorString:String=Append(A=Get CurrentErrorString(), B=Get l_newline() -> (l_newline:String), C=Get l_newError() -> (l_newError:String))
   ) -> (Output_Get:String)
  ]}
 AddErrorSub {nodes [
   n1 = AddErrorSub!() -> (then, NewError:String) {is=title K2Node_FunctionEntry}
   > AddError(self:Self Object Reference, NewError=@n1, Depth=1) {is=title K2Node_CallFunction}
  ]}
 AppendErrorRow {nodes [
   n1 = AppendErrorRow!() -> (then, ErrorIn:"String (by ref)", NewError:String) {is=title K2Node_FunctionEntry}
   > n2 = Branch(Condition=IsEmpty(InString=@n1.NewError))
   n3 = IsEmpty(InString=@n1)
   "Set String (by ref)"!(execute:Exec=@n2.else, Target:"String (by ref)"=@n1, Value:String=Append(A=@n1, B=SelectString(B="\r\n", bPickA=@n3), C=@n1.NewError)) -> (then:Exec) {is=title K2Node_VariableSetRef}
  ]}
 CheckIfAssetClassAllowed {nodes [
   n1 = CheckIfAssetClassAllowed!() -> (then, AssetData:Asset Data Structure, AllowedClasses:Array of Object Soft Class References) {is=title K2Node_FunctionEntry}
   > Set l_classes(l_classes:Array of Object Soft Class References=@n1.AllowedClasses) -> (Output_Get:Array of Object Soft Class References)
   > n2 = Branch(Condition=Is Valid Index!(TargetArray:Array of Object Soft Class References=@n1.AllowedClasses, IndexToTest:int) -> (ReturnValue:bool))
   > n3 = For Each Loop(Array:Array of Object Soft Class References=Get l_classes() -> (l_classes:Array of Object Soft Class References)) -> (Array Element:Object Soft Class Reference)
   Return Node!(execute=@n3.Completed, Allowed:bool=false)
   n4 = GetSoftClassTopLevelAssetPath!(SoftClassReference:Object Soft Class Reference=@n3) -> (ReturnValue:Top Level Asset Path Structure)
   n5 = Break Asset Data!(AssetData=@n1) -> (AssetClassPath)
   > Branch(Condition="Equal Exactly (String)"(
     A=BreakTopLevelAssetPath!(TopLevelAssetPath:"Top Level Asset Path Structure (by ref)"=@n5) -> (PathString:String)
     B=BreakTopLevelAssetPath(TopLevelAssetPath=@n4)
    ))
   > Return Node(Allowed=true)
   Return Node(execute=@n2.else, Allowed=true)
  ]}
 CheckIfObjectClassAllowed {nodes [
   n1 = CheckIfObjectClassAllowed!() -> (then, Object:Object Reference, AllowedClasses:Array of Object Soft Class References) {is=title K2Node_FunctionEntry}
   > Set l_classes(l_classes=@n1.AllowedClasses)
   > n2 = Branch(Condition=Is Valid Index(TargetArray=@n1.AllowedClasses))
   > n3 = For Each Loop(Array=Get l_classes())
   Return Node(execute=@n3.Completed, Allowed=false)
   > n4 = LoadClassAsset_Blocking(AssetClass=@n3)
   > Branch(Condition=ClassIsChildOf(TestClass=Get Class(Object=@n1) -> (ReturnValue:Object Class Reference), ParentClass=@n4))
   > Return Node(Allowed=true)
   Return Node(execute=@n2.else, Allowed=true)
  ]}
 ClearErrorString {nodes [
   ClearErrorString() {is=title K2Node_FunctionEntry}
   > Set CurrentErrorString()
  ]}
 EventGraph {links [], nodes []}
 ExecuteItemClicked {nodes [ExecuteItemClicked!() -> (then, keyState:Modifier Keys State Structure) {is=title K2Node_FunctionEntry}]}
 GetCurrentErrorString {nodes [
   GetCurrentErrorString() {is=title K2Node_FunctionEntry}
   > Return Node!(execute, CurrentErrorString:String=Get CurrentErrorString())
  ]}
 GetCurrentSeverity {nodes [
   GetCurrentSeverity() {is=title K2Node_FunctionEntry}
   > Return Node!(execute, CurrentSeverity:Enum_GART_ReportSeverity Enum=Get CurrentSeverity() -> (CurrentSeverity:Enum_GART_ReportSeverity Enum))
  ]}
 GetReportColorFromSeverity {nodes [
   n1 = GetReportColorFromSeverity!() -> (then, Severity:Enum_GART_ReportSeverity Enum) {is=title K2Node_FunctionEntry}
   > Return Node!(
    execute
    Color:Linear Color Structure=Select!(
     NewEnumerator0:Linear Color Structure=(R=0.2969,G=0.2969,B=0.2969,A=1)
     NewEnumerator1:Linear Color Structure=(R=0.4688,G=0.4033,B=0.0586,A=1)
     NewEnumerator2:Linear Color Structure=(R=0.5729,G=0.0428,B=0,A=1)
     NewEnumerator3:Linear Color Structure=(R=0.224,G=0.0062,B=0,A=1)
     NewEnumerator4:Linear Color Structure
     Index:Enum_GART_ReportSeverity Enum=@n1
    ) -> (ReturnValue:Linear Color Structure)
   )
  ]}
 IncreaseErrorSeverity {nodes [
   n1 = IncreaseErrorSeverity!() -> (then, Severity:"Enum_GART_ReportSeverity Enum (by ref)"=NewEnumerator0, NewSeverity:Enum_GART_ReportSeverity Enum) {is=title K2Node_FunctionEntry}
   > Branch(Condition="integer > integer"(A="To Integer (Byte)"!(InByte:Byte=@n1.NewSeverity) -> (ReturnValue:int=0), B="To Integer (Byte)"(InByte=@n1) -> (ReturnValue=0)))
   > "Set Enum_GART_ReportSeverity Enum (by ref)"!(execute:Exec, Target:"Enum_GART_ReportSeverity Enum (by ref)"=@n1, Value:Enum_GART_ReportSeverity Enum=@n1.NewSeverity) -> (then:Exec) {is=title K2Node_VariableSetRef}
  ]}
 InternalFilterAssetData {nodes [
   n1 = InternalFilterAssetData!() -> (then, AssetData:Asset Data Structure, AllowLoading:bool=~) {is=title K2Node_FunctionEntry}
   > n2 = CheckIfAssetClassAllowed!(
    execute
    self
    AssetData:Asset Data Structure=@n1
    AllowedClasses:Array of Object Soft Class References=Get AllowedClassesToRunOn() -> (AllowedClassesToRunOn:Array of Object Soft Class References)
   ) -> (then, Allowed:bool) {is=title K2Node_CallFunction}
   > n3 = Branch(Condition=@n2)
   > ClearErrorString() {is=title K2Node_CallFunction}
   > n4 = RunFilterLogicOnAssetData!(
    execute
    self
    AssetData:Asset Data Structure=@n1
    AllowLoading:bool=OR Boolean(A=@n1.AllowLoading, B=IsAssetLoaded!(InAssetData:"Asset Data Structure (by ref)"=@n1) -> (ReturnValue:bool))
   ) -> (then, Valid:bool, ReportData:Struct GART Report Data Structure) {is=title K2Node_CallFunction}
   > Set l_valid(l_valid:bool=@n4) -> (Output_Get:bool)
   > n5 = InternalValidatePostFilter!(
    execute
    self
    Data:"Struct GART Report Output Data Structure (by ref)"=Make Struct GART Report Output Data!(
     ReportData_46_#1:Struct GART Report Data Structure=@n4.ReportData
     AssetData_36_#2:Asset Data Structure=@n1
     Object_39_#3:Object Reference=Get l_emptyObject() -> (l_emptyObject:Object Reference)
    ) -> (Struct_GART_ReportOutputData:Struct GART Report Output Data Structure) {is=title K2Node_MakeStruct}
   ) -> (then, DataOut:Struct GART Report Output Data Structure)
   > Return Node!(execute, Valid:bool=Get l_Valid() -> (l_Valid:bool), ReportData:Struct GART Report Output Data Structure=@n5)
   Return Node(execute=@n3.else)
  ]}
 InternalFilterObject {nodes [
   n1 = InternalFilterObject!() -> (then, Object:Object Reference) {is=title K2Node_FunctionEntry}
   > n2 = CheckIfObjectClassAllowed!(execute, self, Object:Object Reference=@n1, AllowedClasses:Array of Object Soft Class References=Get AllowedClassesToRunOn()) -> (then, Allowed:bool) {is=title K2Node_CallFunction}
   > n3 = Branch(Condition=@n2)
   > ClearErrorString()
   > n4 = RunFilterLogicOnObject!(execute, self, object:Object Reference=@n1) -> (then, Valid:bool, ReportData:Struct GART Report Data Structure)
   > Set l_Valid(l_Valid:bool=@n4) -> (Output_Get:bool)
   > n5 = InternalValidatePostFilter(Data=Make Struct GART Report Output Data(
     ReportData_46_#1=@n4.ReportData
     AssetData_36_#2=Make Asset Data!() -> (AssetData:Asset Data Structure) {is=title K2Node_MakeStruct}
     Object_39_#3=@n1
    ))
   > Return Node(Valid=Get l_Valid(), ReportData=@n5)
   Return Node(execute=@n3.else)
  ]}
 InternalValidatePostFilter {nodes [
   n1 = InternalValidatePostFilter!() -> (then, Data:"Struct GART Report Output Data Structure (by ref)") {is=title K2Node_FunctionEntry}
   > n2 = Set members in Struct GART Report Output Data!(
    execute:Exec
    StructRef:"Struct GART Report Output Data Structure (by ref)"=@n1
    Report_44_#4:Soft Object Reference=Make Soft Reference!(Input:GART Report Base Object Reference=Self-Reference()) -> (Output:GART Report Base Soft Object Reference)
   ) -> (then:Exec, StructOut:"Struct GART Report Output Data Structure (by ref)") {is=title K2Node_SetFieldsInStruct}
   > n3 = InternalValidateReportColor!(execute, self, Data:"Struct GART Report Output Data Structure (by ref)"=@n2) -> (then, DataOut:Struct GART Report Output Data Structure)
   > n4 = InternalValidateReportDescLong!(execute, self, Data:"Struct GART Report Output Data Structure (by ref)"=@n3) -> (then, DataOut:Struct GART Report Output Data Structure)
   > Return Node!(execute, DataOut:Struct GART Report Output Data Structure=@n4)
  ]}
 InternalValidateReportColor {nodes [
   n1 = InternalValidateReportColor!() -> (then, Data:"Struct GART Report Output Data Structure (by ref)") {is=title K2Node_FunctionEntry}
   > Set l_reportData(l_reportData:Struct GART Report Output Data Structure=@n1) -> (Output_Get:Struct GART Report Output Data Structure)
   n2 = Break Struct GART Report Output Data!(Struct_GART_ReportOutputData:"Struct GART Report Output Data Structure (by ref)"=Get l_reportData() -> (l_reportData:Struct GART Report Output Data Structure)) -> (
    ReportData_46_#1:Struct GART Report Data Structure
    AssetData_36_#2:Asset Data Structure
    Object_39_#3:Object Reference
    Report_44_#4:Soft Object Reference
   ) {is=title K2Node_BreakStruct}
   n3 = Make Struct GART Report Data!(Severity_46_#1=NewEnumerator0) -> (Struct_GART_ReportData)
   > n4 = Branch(Condition="Equal (LinearColor)"(
     A:Linear Color Structure=Break Struct GART Report Data!(Struct_GART_ReportData:"Struct GART Report Data Structure (by ref)"=@n2) -> (
      DisplayColorOverride_40_#5:Linear Color Structure
      Severity_46_#1:Enum_GART_ReportSeverity Enum=~
     ) {is=title K2Node_BreakStruct}
     B:Linear Color Structure=Break Struct GART Report Data(Struct_GART_ReportData=@n3) -> (Severity_46_#1=~)
    ) -> (ReturnValue:bool=~) {is=title K2Node_PromotableOperator})
   n5 = Break Struct GART Report Output Data(Struct_GART_ReportOutputData=Get l_reportData())
   > n6 = GetReportColorFromSeverity!(
    execute
    self
    Severity:Enum_GART_ReportSeverity Enum=Break Struct GART Report Data!(Struct_GART_ReportData=@n5) -> (Severity_46_#1=~)
   ) -> (then, Color:Linear Color Structure) {is=title K2Node_CallFunction}
   > n7 = Set members in Struct GART Report Data!(
    execute:Exec
    StructRef:"Struct GART Report Data Structure (by ref)"=@n5
    DisplayColorOverride_40_#5:Linear Color Structure=@n6
   ) -> (then:Exec, StructOut:"Struct GART Report Data Structure (by ref)") {is=title K2Node_SetFieldsInStruct}
   > n8 = Set members in Struct GART Report Output Data!(execute, StructRef=Get l_reportData(), ReportData_46_#1:Struct GART Report Data Structure=@n7) -> (then, StructOut)
   > Return Node(DataOut:"Struct GART Report Output Data Structure (by ref)"=@n8)
   Return Node(execute=@n4.else, DataOut=Get l_reportData())
  ]}
 InternalValidateReportDescLong {nodes [
   n1 = InternalValidateReportDescLong!() -> (then, Data:"Struct GART Report Output Data Structure (by ref)") {is=title K2Node_FunctionEntry}
   > Set l_reportData(l_reportData=@n1)
   > n2 = Branch(Condition=IsEmpty(InString=Get CurrentErrorString()))
   > Return Node(DataOut=Get l_reportData())
   n3 = Set members in Struct GART Report Data!(
    execute=@n2.else
    StructRef=Break Struct GART Report Output Data(Struct_GART_ReportOutputData=Get l_reportData())
    DisplayDescLong_30_#6:String=Get CurrentErrorString()
   ) -> (then, StructOut)
   > n4 = Set members in Struct GART Report Output Data(StructRef=Get l_reportData(), ReportData_46_#1=@n3)
   > Return Node(DataOut=@n4)
  ]}
 InternalValidateSeverity {nodes [
   n1 = InternalValidateSeverity!() -> (then, Data:"Struct GART Report Output Data Structure (by ref)") {is=title K2Node_FunctionEntry}
   > Set l_reportData(l_reportData=@n1)
   > n2 = Branch(Condition="Equal (Enum)"(A:Enum_GART_ReportSeverity Enum=Get CurrentSeverity(), B:Enum_GART_ReportSeverity Enum=NewEnumerator4))
   > Return Node(DataOut=Get l_reportData())
   n3 = Set members in Struct GART Report Data!(
    execute=@n2.else
    StructRef=Break Struct GART Report Output Data(Struct_GART_ReportOutputData=Get l_reportData())
    Severity_46_#1:Enum_GART_ReportSeverity Enum=Get CurrentSeverity()
   ) -> (then, StructOut)
   > n4 = Set members in Struct GART Report Output Data(StructRef=Get l_reportData(), ReportData_46_#1=@n3)
   > Return Node(DataOut=@n4)
  ]}
 RunFilterLogicOnAssetData {nodes [
   RunFilterLogicOnAssetData() -> (AllowLoading=~) {is=title K2Node_FunctionEntry}
   > Return Node!(execute, Valid:bool=false, ReportData:Struct GART Report Data Structure)
  ]}
 RunFilterLogicOnObject {nodes [
   RunFilterLogicOnObject!() -> (then, object:Object Reference) {is=title K2Node_FunctionEntry}
   > Return Node(Valid=false)
  ]}
 SetSeverity {nodes [
   SetSeverity!() -> (then, NewSeverity:Enum_GART_ReportSeverity Enum, OnlyIfHigher:bool=true) {is=title K2Node_FunctionEntry}
   n1 = Get OnlyIfHigher() -> (OnlyIfHigher:bool)
   n2 = "Not Equal (Enum)"!(
    A:Enum_GART_ReportSeverity Enum=Get NewSeverity() -> (NewSeverity:Enum_GART_ReportSeverity Enum)
    B:Enum_GART_ReportSeverity Enum=NewEnumerator4
   ) -> (ReturnValue:bool)
   n3 = "integer > integer"(A="To Integer (Byte)"(InByte=Get NewSeverity()) -> (ReturnValue=0), B="To Integer (Byte)"(InByte=Get CurrentSeverity()) -> (ReturnValue=0))
   > Branch(Condition=OR Boolean(A=AND Boolean(A=@n3, B=@n2), B=NOT Boolean(A=@n1)))
   > "Set Enum_GART_ReportSeverity Enum (by ref)"(Target=Get CurrentSeverity(), Value=Get NewSeverity())
  ]}
]
variables [
 AllowedClassesToRunOn {type="array<Class>", default []}
 CurrentErrorString {type=string, default=""}
 CurrentSeverity {type=Enum_GART_ReportSeverity, default=None}
 ExposedFunctions {type="array<GART_ExposedFunction_Struct>", default []}
]
